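# Base cloth material for the KneeLens backpack, layering an animated reactive screen over the Anime_Parent_2023 character shader.
# Custom primitive data supplies a timestamp, duration and frame selection, driving timelines for tint, scanline and CRT-style closing-line effects.
# The screen picks between two texture halves, adding UV distortion, flipbook noise and blue screen-blend tint, output as exposure-compensated emissive.
# Fixed-loop trigger timing nodes remain with unconnected outputs.
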Material /BRCosmetics/Accessories/FORT_Backpacks/Backpack_F_MED_KneeLens/Materials/M_KneeLens_Backpack_Base
graph {
 properties {
  bAutomaticallySetUsageInEditor=false
  bUseMaterialAttributes=true
  bUsedWithMeshDeformer=true
  bUsedWithMorphTargets=true
  bUsedWithSkeletalMesh=true
  bUsedWithStaticLighting=true
  pixelDepthOffsetMode=PDOM_Legacy
  shadingModel=MSM_Cloth
  usedShadingModels=Cloth
 }
 nodes [
  n1 = MaterialFunctionCall!(CustomUV) {materialFunction=/Game/Global/Materials/Character_Parent_2023/Anime_Parent_2023/MaterialFunctions/MFC_AnimeShader_2023_Pre}
  n2 = MaterialFunctionCall!(CustomLightVector, OverrideNormal_WS, STT=@n1.STT, Normal_Map=@n1.STT, DFL=@n1.STT, FX=@n1.STT) {materialFunction=/Game/Global/Materials/Character_Parent_2023/Anime_Parent_2023/MaterialFunctions/MFC_AnimeShader_2023_Calculations}
  n3 = MaterialFunctionCall!(
   Shading_Gradient=@n2.Shading_Gradient
   Threshold_Gradient=@n2.Shading_Gradient
   Specular_Gradient=@n2.Shading_Gradient
   Sheen_Mask=@n2.Shading_Gradient
   Thickness=@n2.Shading_Gradient
   DFL_Mask=@n2.Shading_Gradient
   FX=@n2.Shading_Gradient
   Metallic_Gradient=@n2.Shading_Gradient
   RimLight_Gradient=@n2.Shading_Gradient
   NormalBuffer=@n2.Shading_Gradient
   --------
   CINE_LightSources=@n2.Shading_Gradient
   CINE_LightColor=@n2.Shading_Gradient
   ---------
   Color_Lit=@n1.Color_Lit
   Color_Shaded=@n1.Color_Lit
   SRM=@n1.Color_Lit
   Emissive=@n1.Color_Lit
   DFL_Color=@n1.Color_Lit
   Specular_Color=@n1.Color_Lit
   RimLight_Color=@n1.Color_Lit
  ) {materialFunction=/Game/Global/Materials/Character_Parent_2023/Anime_Parent_2023/MaterialFunctions/MFC_AnimeShader_2023_Post}
  n4 = GetMaterialAttributes(None=@n3.Result) {attributeGetTypes [#1]}
  n5 = LinearInterpolate(A=@n4.Emissive Color, B=NamedRerouteUsage(), Alpha=Multiply(A=NamedRerouteUsage(), B=NamedRerouteUsage()))
  n6 = MaterialFunctionCall!(Blend=@n5, Base=VectorParameter() {group=KneeLens, parameterName=BlueFilter_Color, defaultValue=<rgba 0 0.0591 0.2 1>}.RGB) {materialFunction=/Engine/Functions/Engine_MaterialFunctions03/Blends/Blend_Screen}
  n7 = Multiply(A=ScalarParameter() {defaultValue=0.5, group=KneeLens, parameterName=BlueFilter_Intensity}, B=NamedRerouteUsage())
  n8 = Multiply(A=NamedRerouteUsage(), B=ScalarParameter() {defaultValue=1, group=KneeLens, parameterName=Scanlines_Intensity}) {constB=2}
  n9 = NamedRerouteUsage()
  n10 = NamedRerouteUsage()
  n11 = ScalarParameter() {defaultValue=1, group=KneeLens, parameterName=EndScreenLine_Intensity}
  n12 = Add(
   A=Add(A=LinearInterpolate(A=@n5, B=@n6.Result, Alpha=@n7), B=Multiply(A=@n8, B=@n9))
   B=Multiply(A=Multiply(A=@n10, B=@n11) {constB=2}, B=NamedRerouteUsage()) {constB=2}
  )
  BlendMaterialAttributes(
   A=@n3.Result
   B=SetMaterialAttributes(MaterialAttributes=@n4.MaterialAttributes, Emissive Color=EyeAdaptationInverse(LightValueInput=@n12).EyeAdaptationInverse) {attributeSetTypes [#1]}
   Alpha=NamedRerouteUsage()
  )
  n13 = NamedRerouteUsage()
  Divide(A=@n13, B=ScalarParameter() {defaultValue=0.3, parameterName=ScanlinesFlipSpeed})
  n14 = NamedRerouteDeclaration(None=Fmod(A=Time(), B=ScalarParameter() {defaultValue=40, group=KneeLens, parameterName=LoopDuration})) {name=FmodTime, nodeColor=<rgba 0 1 0.3438 1>}
  n15 = ScalarParameter() {defaultValue=15, group=KneeLens, parameterName=TriggerTime}
  n16 = ScalarParameter() {defaultValue=10, group=KneeLens, parameterName=TriggerDuration}
  Min(A=OneMinus(None=Step(Y=@n14, X=@n15)), B=Step(Y=@n14, X=Add(A=@n15, B=@n16) {constB=10}))
  n17 = ScalarParameter() {defaultValue=1, group=KneeLens, parameterName=TriggerOffset}
  n18 = NamedRerouteDeclaration(None=Subtract(A=NamedRerouteUsage(), B=@n17) {constB=0.1}) {name=FmodOffsetTime, nodeColor=<rgba 1 0 0.5625 1>}
  n19 = ScalarParameter() {defaultValue=15, group=KneeLens, parameterName=TriggerTime}
  n20 = Subtract(A=ScalarParameter() {defaultValue=10, group=KneeLens, parameterName=TriggerDuration}, B=@n17) {constB=0.1}
  Min(A=OneMinus(None=Step(Y=@n18, X=@n19)), B=Step(Y=@n18, X=Add(A=@n19, B=@n20) {constB=10}))
  n21 = TextureObjectParameter() {
   group=KneeLens
   parameterName=ScreenFrames
   texture=/BRCosmetics/Accessories/FORT_Backpacks/Backpack_F_MED_KneeLens/Textures/T_KneeLens_ScreenFrames
  }
  n22 = TextureSample(UVs=NamedRerouteUsage(), Tex=@n21) {
   samplerSource=SSM_Clamp_WorldGroupSettings
   texture=&T_KneeLens_ScreenFrames
  }
  NamedRerouteDeclaration(None=@n22.RGB) {name=FrameA, nodeColor=<rgba 0.875 1 0 1>}
  n23 = Multiply(A=Time() {period=10}, B=ScalarParameter() {defaultValue=0.75, parameterName=NoiseSpeed})
  n24 = Constant() {r=10}
  n25 = Constant() {r=3}
  n26 = TextureObjectParameter() {
   group=KneeLens
   parameterName=NoiseFlip
   samplerType=SAMPLERTYPE_Masks
   texture=/BRCosmetics/Accessories/FORT_Backpacks/Backpack_F_MED_KneeLens/Textures/T_KneeLens_NoiseMask
  }
  NamedRerouteDeclaration(None=MaterialFunctionCall("Animation  Phase (0-1)"=Fmod(A=@n23, B=@n24), Number of Rows=@n25, Number of Columns=@n25, Texture=@n26, UVs=NamedRerouteUsage()) {materialFunction=&FlipBook}.Result) {name=NoiseFlipbook, nodeColor=<rgba 1 0 0.2344 1>}
  n27 = Saturate(None=SquareRoot(None=@n13))
  n28 = Constant() {r=3}
  n29 = TextureObjectParameter() {
   group=KneeLens
   parameterName=NoiseFlip
   samplerType=SAMPLERTYPE_Masks
   texture=&T_KneeLens_NoiseMask
  }
  n30 = NamedRerouteUsage()
  n31 = StaticBool() {value=true}
  NamedRerouteDeclaration(None=ComponentMask(None=MaterialFunctionCall("Animation  Phase (0-1)"=@n27, Number of Rows=@n28, Number of Columns=@n28, Texture=@n29, UVs=@n30, "Clamp Anim (see tooltip)"=@n31) {materialFunction=&FlipBook}.Result) {b=true}) {name=ScanlinesAnimated, nodeColor=<rgba 0 1 0.25 1>}
  n32 = NamedRerouteUsage()
  n33 = AppendVector(
   A=ScalarParameter() {group=KneeLens, parameterName=DistortTilingX, sortPriority=4}
   B=ScalarParameter() {defaultValue=1, group=KneeLens, parameterName=DistortTilingY, sortPriority=4}
  )
  n34 = TextureSampleParameter2D(UVs=Panner(Coordinate=Multiply(A=@n32, B=@n33)) {bFractionalPart=true, speedX=0.015, speedY=0.384}) {
   group=KneeLens
   parameterName=DistortTexture
   samplerType=SAMPLERTYPE_Normal
   sortPriority=4
   texture=/Game/Global/Textures/Default/Noise/64x64Noise
  }
  n35 = AppendVector(A=@n34.R, B=Constant())
  n36 = Time()
  n37 = Sine(None=@n36) {period=0.355}
  n38 = Sine(None=@n36) {period=0.8755}
  n39 = Power(Base=Saturate(None=Add(A=@n37, B=@n38))) {constExponent=10}
  n40 = VertexInterpolator(VS=Multiply(A=ScalarParameter() {defaultValue=0.1, group=KneeLens, parameterName=DistortScale, sortPriority=4}, B=ComponentMask(None=@n39) {r=true}))
  NamedRerouteDeclaration(None=Add(A=NamedRerouteUsage(), B=Multiply(A=@n35, B=@n40.PS))) {name=DistortedUVs, nodeColor=<rgba 0 0.7188 1 1>}
  n41 = ComponentMask(None=NamedRerouteUsage()) {g=true}
  n42 = SquareRoot(None=NamedRerouteUsage())
  n43 = LinearInterpolate(Alpha=@n42) {constA=1, constB=0.5}
  n44 = LinearInterpolate(Alpha=@n42) {constB=0.5}
  NamedRerouteDeclaration(None=Subtract(A=Step(Y=@n41, X=@n43), B=Step(Y=@n41, X=@n44) {constX=0})) {name=EndScreenMask, nodeColor=<rgba 1 0 0.0469 1>}
  n45 = ComponentMask(None=NamedRerouteUsage()) {g=true}
  n46 = ScalarParameter() {defaultValue=0.0075, group=KneeLens, parameterName=EndScreenLineThickness}
  n47 = Add(B=@n46) {constA=0.5, constB=0.5}
  n48 = Subtract(B=@n46) {constA=0.5}
  NamedRerouteDeclaration(None=Subtract(A=Step(Y=@n45, X=@n47) {constX=0.5075}, B=Step(Y=@n45, X=@n48) {constX=0.4925})) {name=EndScreenLine, nodeColor=<rgba 0 1 0.5078 1>}
  n49 = OneMinus(None=NamedRerouteUsage())
  NamedRerouteDeclaration(None=Min(A=NamedRerouteUsage(), B=SmoothStep(Value=@n49))) {name=EndScreenLineState, nodeColor=<rgba 1 0.6563 0 1>}
  n50 = TextureSample(UVs=NamedRerouteUsage(), Tex=@n21) {
   samplerSource=SSM_Clamp_WorldGroupSettings
   texture=&T_KneeLens_ScreenFrames
  }
  NamedRerouteDeclaration(None=@n50.RGB) {name=FrameB, nodeColor=<rgba 1 0 0.3984 1>}
  n51 = ScalarParameter() {bUseCustomPrimitiveData=true, group=KneeLens, parameterName=TimeStamping_CPD, primitiveDataIndex=1}
  n52 = Subtract(A=Time(), B=@n51)
  n53 = ScalarParameter() {defaultValue=1, group=KneeLens, parameterName=TriggerOffset}
  n54 = NamedRerouteDeclaration(None=Saturate(None=Divide(A=@n52, B=@n53) {constB=0.5})) {name=PreAnimationTimeline, nodeColor=<rgba 1 0 0.1172 1>}
  n55 = OneMinus(None=@n54)
  n56 = OneMinus(None=Max(A=Step(Y=@n54) {constX=0}, B=Step(Y=@n55) {constX=0}))
  n57 = ScalarParameter() {bUseCustomPrimitiveData=true, defaultValue=10, group=KneeLens, parameterName=ReactivityDuration_CPD, primitiveDataIndex=2}
  n58 = Divide(A=@n52, B=@n57) {constB=0.5}
  n59 = NamedRerouteDeclaration(None=Multiply(A=Saturate(None=@n58), B=Saturate(None=@n51))) {name=AnimationTimeline, nodeColor=<rgba 1 0.4922 0 1>}
  n60 = ScalarParameter() {bUseCustomPrimitiveData=true, defaultValue=10, group=KneeLens, parameterName=ReactivityDuration_CPD, primitiveDataIndex=2}
  n61 = Divide(A=ScalarParameter() {defaultValue=1, group=KneeLens, parameterName=TriggerOffset}, B=@n60)
  n62 = OneMinus(None=@n59)
  n63 = OneMinus(None=Max(A=Step(Y=@n59, X=@n61) {constX=0}, B=Step(Y=@n62) {constX=0}))
  NamedRerouteDeclaration(None=Max(
    A=NamedRerouteDeclaration(None=@n56) {name=PreAnimationState, nodeColor=<rgba 1 0 0.1875 1>}
    B=NamedRerouteDeclaration(None=@n63) {name=AnimationState, nodeColor=<rgba 0 0.25 1 1>}
   )) {name=TintState, nodeColor=<rgba 0 1 0.8359 1>}
  n64 = Time()
  n65 = ScalarParameter() {bUseCustomPrimitiveData=true, group=KneeLens, parameterName=TimeStamping_CPD, primitiveDataIndex=1}
  n66 = ScalarParameter() {bUseCustomPrimitiveData=true, defaultValue=10, group=KneeLens, parameterName=ReactivityDuration_CPD, primitiveDataIndex=2}
  n67 = Subtract(A=@n64, B=Add(A=@n65, B=@n66))
  n68 = ScalarParameter() {defaultValue=2, group=KneeLens, parameterName=EndScreenLineDuration}
  n69 = NamedRerouteDeclaration(None=Saturate(None=Divide(A=@n67, B=@n68) {constB=0.5})) {name=PostAnimationTimeline, nodeColor=<rgba 0.1719 0 1 1>}
  n70 = Step(Y=@n69) {constX=0}
  n71 = Step(Y=OneMinus(None=@n69)) {constX=0}
  NamedRerouteDeclaration(None=OneMinus(None=Max(A=@n70, B=@n71))) {name=PostAnimationState, nodeColor=<rgba 1 0 0.0703 1>}
  n72 = ScalarParameter() {defaultValue=0.075, group=KneeLens, parameterName=EndScreenMaskDuration}
  n73 = Subtract(A=@n64, B=Add(A=@n65, B=Subtract(A=@n66, B=@n72)))
  n74 = NamedRerouteDeclaration(None=Saturate(None=Divide(A=@n73, B=@n72) {constB=0.5})) {name=CloseAnimationTimeline, nodeColor=<rgba 1 0 0.3281 1>}
  n75 = Step(Y=@n74) {constX=0}
  n76 = Step(Y=OneMinus(None=@n74)) {constX=0}
  NamedRerouteDeclaration(None=OneMinus(None=Max(A=@n75, B=@n76))) {name=CloseAnimationState, nodeColor=<rgba 0 0.25 1 1>}
  n77 = MaterialFunctionCall(R=NamedRerouteUsage(), G=NamedRerouteUsage(), B=NamedRerouteUsage(), A=NamedRerouteUsage()) {materialFunction=&AppendMany}
  n78 = MaterialFunctionCall(R=NamedRerouteUsage(), G=NamedRerouteUsage(), B=NamedRerouteUsage(), A=NamedRerouteUsage()) {materialFunction=&AppendMany}
  n79 = TextureCoordinate()
  n80 = AppendVector(A=Add(A=ComponentMask(None=@n79) {r=true}) {constB=0.3}, B=ComponentMask(None=@n79) {g=true})
  NamedRerouteDeclaration(None=Add(
    A=MaterialFunctionCall!(Vector4=@n77.RGBA, MaximumNumberOfDigits, UVs, DebugTextLocation RG_UpperRight BA_LowerLeft, Component Spacing) {materialFunction=/Engine/Functions/Engine_MaterialFunctions02/Utility/DebugFloat4Values}.ColorCodedOutput
    B=MaterialFunctionCall(Vector4=@n78.RGBA, UVs=@n80) {materialFunction=&DebugFloat4Values}.ColorCodedOutput
   )) {name=Debug, nodeColor=<rgba 0.1719 0 1 1>}
  n81 = NamedRerouteUsage()
  n82 = Multiply(A=ComponentMask(None=@n81) {r=true}) {constB=0.5}
  n83 = ComponentMask(None=@n81) {g=true}
  NamedRerouteDeclaration(None=AppendVector(A=Add(A=@n82) {constB=0.5}, B=@n83)) {name=FrameA_UVs, nodeColor=<rgba 1 0 0.5859 1>}
  NamedRerouteDeclaration(None=AppendVector(A=@n82, B=@n83)) {name=FrameB_UVs, nodeColor=<rgba 0.8281 1 0 1>}
  n84 = NamedRerouteUsage()
  n85 = Subtract(A=Step(Y=@n84), B=Step(Y=@n84) {constX=0})
  NamedRerouteDeclaration(None=Min(A=ComponentMask(None=@n85) {r=true}, B=ComponentMask(None=@n85) {g=true})) {name=ScreenMask, nodeColor=<rgba 0 0.7891 1 1>}
  n86 = TextureCoordinate()
  n87 = VectorParameter() {group=KneeLens, parameterName=ScreenUVs_Scale, defaultValue=<rgba 7.12 9.5 0 1>}
  n88 = AppendVector(A=@n87.R, B=@n87.R)
  n89 = VectorParameter() {group=KneeLens, parameterName=ScreenUVs_Offset, defaultValue=<rgba -0.105 -3.68 0 1>}
  NamedRerouteDeclaration(None=Add(A=Multiply(A=@n86, B=@n88) {constB=4}, B=AppendVector(A=@n89.R, B=@n89.R)) {constB=0}) {name=ScreenUVs, nodeColor=<rgba 0 1 0.4844 1>}
  n90 = Desaturation(
   None=LinearInterpolate(
    A=NamedRerouteUsage()
    B=NamedRerouteUsage()
    Alpha=ScalarParameter() {bUseCustomPrimitiveData=true, group=KneeLens, parameterName=FrameSelection_CPD, primitiveDataIndex=3}
   )
   Fraction=ScalarParameter() {group=KneeLens, parameterName=ScreenDesaturation}
  )
  n91 = ScalarParameter() {defaultValue=0.6, group=KneeLens, parameterName=ScreenBrightness}
  n92 = NamedRerouteUsage()
  n93 = ScalarParameter() {defaultValue=1, group=KneeLens, parameterName=Noise_Intensity}
  NamedRerouteDeclaration(None=Add(A=Multiply(A=@n90, B=@n91), B=Multiply(A=@n92, B=@n93) {constB=2})) {name=ScreenOutput, nodeColor=<rgba 1 0 0.9844 1>}
  n94 = Multiply(A=Time() {period=10}, B=Constant() {r=0.01})
  n95 = TextureSample(UVs=Fmod(A=ComponentMask(None=@n94) {r=true}, B=Constant() {r=10})) {
   samplerSource=SSM_Wrap_WorldGroupSettings
   samplerType=SAMPLERTYPE_LinearGrayscale
   texture=/BRCosmetics/Global/Textures/Default/Noise/T_BlueNoise
  }
  n96 = Constant() {r=3}
  n97 = Floor(None=Multiply(A=@n95.R, B=@n96) {constB=3})
  n98 = NamedRerouteUsage()
  NamedRerouteDeclaration(None=If(
    A=Divide(A=@n97, B=@n96) {constB=3}
    "A > B"=Constant()
    "A == B"=ComponentMask(None=@n98) {g=true}
    "A < B"=ComponentMask(None=@n98) {r=true}
   ) {constB=0.3333, equalsThreshold=0.001}) {name=OutputNoise, nodeColor=<rgba 0 0.4141 1 1>}
  n99 = Subtract(A=NamedRerouteUsage(), B=ScalarParameter() {defaultValue=15, group=KneeLens, parameterName=TriggerTime})
  n100 = ScalarParameter() {defaultValue=0.075, group=KneeLens, parameterName=EndScreenMaskDuration}
  n101 = Subtract(A=ScalarParameter() {defaultValue=10, group=KneeLens, parameterName=TriggerDuration}, B=@n100)
  Saturate(None=Divide(A=Subtract(A=@n99, B=@n101), B=@n100))
  n102 = Subtract(A=NamedRerouteUsage(), B=ScalarParameter() {defaultValue=15, group=KneeLens, parameterName=TriggerTime})
  n103 = Subtract(
   A=ScalarParameter() {defaultValue=10, group=KneeLens, parameterName=TriggerDuration}
   B=ScalarParameter() {defaultValue=0.075, group=KneeLens, parameterName=EndScreenMaskDuration}
  )
  Saturate(None=Divide(A=Subtract(A=@n102, B=@n103), B=ScalarParameter() {defaultValue=2, group=KneeLens, parameterName=EndScreenLineDuration}))
 ]
}
